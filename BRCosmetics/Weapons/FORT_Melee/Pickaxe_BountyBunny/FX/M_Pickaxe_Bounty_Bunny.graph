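# Translucent FX material for the Bounty Bunny pickaxe, flagged for beam trails and Niagara ribbons, drawing a flowing multicolor energy trail.
# Two panning noise samples blend three color parameters, oversaturated through negative desaturation before emissive adjustment.
# Opacity sums normal-distorted smoke and energy, a trail mask and a gradient, scaled by particle alpha and two U-axis sphere masks.
# Depth fade, time-of-day scaling and optional view-angle fade refine that opacity.
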Material /BRCosmetics/Weapons/FORT_Melee/Pickaxe_BountyBunny/FX/M_Pickaxe_Bounty_Bunny
graph {
 properties {
  bUsedWithBeamTrails=true
  bUsedWithNiagaraRibbons=true
  blendMode=BLEND_Translucent
  pixelDepthOffsetMode=PDOM_Legacy
  usedShadingModels=Default Lit
 }
 nodes [
  n1 = TextureCoordinate() {uTiling=0.3, vTiling=0.6}
  n2 = LinearInterpolate(
   A=VectorParameter() {parameterName=Color, defaultValue=<rgba 0.0793 1 0.3024 1>}.RGB
   B=VectorParameter() {parameterName=Color_1, defaultValue=<rgba 0.1269 0.716 1 1>}.RGB
   Alpha=TextureSample(UVs=Panner(Coordinate=@n1) {speedX=-0.7}) {texture=&T_NoisePack_01}.R
  )
  n3 = VectorParameter() {parameterName=Color_2, defaultValue=<rgba 0.8157 0.3707 1 1>}
  n4 = TextureSample(UVs=Panner(Coordinate=TextureCoordinate() {uTiling=0.12, vTiling=0.4}) {speedX=-1.2}) {texture=&T_NoisePack_01}
  MaterialFunctionCall(Emissive=Desaturation(None=LinearInterpolate(A=@n2, B=@n3.RGB, Alpha=@n4.G), Fraction=Constant() {r=-1.5})) {materialFunction=&MF_FX_EmissiveAdjustments}
  n5 = ScalarParameter() {defaultValue=0.85, parameterName=UV_U_1}
  n6 = ScalarParameter() {defaultValue=1, parameterName=UV_V_1}
  n7 = TextureCoordinate() {uTiling=0.4, vTiling=0.55}
  n8 = ComponentMask(None=TextureSample(UVs=Panner(Coordinate=@n7) {speedX=-0.4, speedY=0.6}) {samplerType=SAMPLERTYPE_Normal, texture=&T_Weap_Distort_N_02}.RGB) {g=true, r=true}
  n9 = Add(A=Multiply(A=AppendVector(A=@n5, B=@n6), B=TextureCoordinate()), B=Multiply(A=@n8) {constB=0.2})
  n10 = AppendVector(A=ScalarParameter() {defaultValue=0.36, parameterName=Speed_U_1}, B=ScalarParameter() {parameterName=Speed_V_1})
  n11 = Multiply(A=@n8) {constB=0.22}
  n12 = Multiply(
   A=TextureCoordinate()
   B=AppendVector(A=ScalarParameter() {defaultValue=0.52, parameterName=UV_U}, B=ScalarParameter() {defaultValue=1, parameterName=UV_V})
  )
  n13 = ScalarParameter() {defaultValue=0.5, parameterName=Speed_U}
  n14 = ScalarParameter() {parameterName=Speed_V}
  n15 = TextureSampleParameter2D(UVs=Panner(Coordinate=Add(A=@n11, B=@n12), Speed=AppendVector(A=@n13, B=@n14))) {parameterName=Texture, texture=/Game/Effects/Fort_Effects/Textures/Energy/T_EnergyFlow_Packed}
  n16 = Multiply(A=@n15.R, B=@n15.R)
  n17 = TextureCoordinate()
  n18 = AppendVector(A=ScalarParameter() {defaultValue=0.2, parameterName=UV_U_2}, B=ScalarParameter() {defaultValue=1, parameterName=UV_V_2})
  n19 = ScalarParameter() {defaultValue=1, parameterName=Speed_U_2}
  n20 = ScalarParameter() {parameterName=Speed_V_2}
  n21 = TextureSample(UVs=Panner(Coordinate=Multiply(A=@n17, B=@n18), Speed=AppendVector(A=@n19, B=@n20))) {texture=/Game/Effects/Fort_Effects/Textures/Smoke/T_Lobber_Trail_Mask}
  n22 = Add(
   A=TextureSample(UVs=Panner(Coordinate=@n9, Speed=@n10)) {samplerType=SAMPLERTYPE_Grayscale, texture=&T_Smoke_Trail_Soft}.R
   B=Add(A=Multiply(A=@n16) {constB=2}, B=Multiply(A=@n21.R) {constB=0.6})
  )
  n23 = Multiply(A=TextureSample() {texture=/Game/Effects/Fort_Effects/Textures/Grad/gradient_healed}.R) {constB=2}
  n24 = MaterialFunctionCall(Float2=TextureCoordinate()) {materialFunction=&BreakOutFloat2Components}
  n25 = ScalarParameter() {defaultValue=0.5, parameterName=SphereMaskA_BValue}
  n26 = ScalarParameter() {defaultValue=0.5, parameterName=SphereMaskA_Radius}
  n27 = MaterialFunctionCall(Float2=TextureCoordinate()) {materialFunction=&BreakOutFloat2Components}
  n28 = ScalarParameter() {parameterName=SphereMaskB_BValue}
  n29 = ScalarParameter() {defaultValue=0.85, parameterName=SphereMaskB_Radius}
  n30 = Multiply(
   A=Multiply(A=ParticleColor().A, B=Add(A=@n22, B=@n23))
   B=Multiply(
    A=SphereMask(A=@n24.R, B=@n25, Radius=@n26) {attenuationRadius=0.5, hardnessPercent=0}
    B=SphereMask(A=@n27.R, B=@n28, Radius=@n29) {attenuationRadius=0.85, hardnessPercent=0}
   )
  )
  n31 = ScalarParameter() {defaultValue=1, parameterName=Opacity_Day}
  n32 = ScalarParameter() {defaultValue=1, parameterName=Opacity_Night}
  n33 = ScalarParameter() {defaultValue=8, parameterName=FadeSensitivity}
  Multiply(
   A=Multiply(
    A=DepthFade(Opacity=@n30) {fadeDistanceDefault=30}
    B=MaterialFunctionCall(Color_Morning=@n31, Color_Day=@n31, Color_Evening=@n32, Color_Night=@n32) {materialFunction=&Color_Time_of_Day}.Result
   )
   B=LinearInterpolate(
    B=MaterialFunctionCall("FadeSensitivity (optional)"=@n33) {materialFunction=&FadeBasedOnViewAngle}.Result
    Alpha=ScalarParameter() {parameterName=Use FadebyViewDistance}
   ) {constA=1}
  )
 ]
}
