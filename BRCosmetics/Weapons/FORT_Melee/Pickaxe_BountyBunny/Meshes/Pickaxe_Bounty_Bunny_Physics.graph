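# Physics asset for the Bounty Bunny two-handed pickaxe, apparently for secondary motion on its dynamic leaf and mushroom bones.
# A long capsule covers the handle, with smaller capsules on eight leaf bones and two mushroom bones.
# Each leaf is constrained directly to the handle with twist locked, and the eighth also locks one swing axis.
# The mushrooms form a two-link chain from the handle with tighter angle settings than the leaves.
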PhysicsAsset /BRCosmetics/Weapons/FORT_Melee/Pickaxe_BountyBunny/Meshes/Pickaxe_Bounty_Bunny_Physics
bodies {
 Melee_TwoHanded_Handle [{shapeType=Capsule, length=99.3868, radius=3.4672, center=<0.5636 -1.1352 20.1682>}]
 dyn_leaves_01 [{shapeType=Capsule, length=3.7957, radius=2.6242, center=<0.7884 -0.0001 0>, rotation=<rot 90 1.2658 1.2656>}]
 dyn_leaves_02 [{shapeType=Capsule, length=3.7957, radius=2.6242, center=<0.6384 0 0.2575>, rotation=<rot 90 -0.02 -0.0202>}]
 dyn_leaves_03 [{shapeType=Capsule, length=3.7957, radius=2.6242, center=<0.5402 0 0>, rotation=<rot 90 0.2371 0.2374>}]
 dyn_leaves_04 [{shapeType=Capsule, length=3.7957, radius=2.6242, center=<0.6697 0 0>, rotation=<rot 90 0.0197 0.0195>}]
 dyn_leaves_05 [{shapeType=Capsule, length=3.7957, radius=2.6242, center=<0.663 0 0>, rotation=<rot 90 -10.025 -10.0249>}]
 dyn_leaves_06 [{shapeType=Capsule, length=3.7957, radius=2.6242, center=<0.5343 0 0>, rotation=<rot 90 -0.0014 -0.0016>}]
 dyn_leaves_07 [{shapeType=Capsule, length=3.7957, radius=2.6242, center=<0.5009 0 0>, rotation=<rot 90 -1.8985 -1.8984>}]
 dyn_leaves_08 [{shapeType=Capsule, length=3.7957, radius=2.6242, center=<0.4953 0 0>, rotation=<rot 90 12.9074 7.909>}]
 dyn_mushroom_01 [{shapeType=Capsule, length=4.8748, radius=3.4431, center=<1.6736 0 0>, rotation=<rot 90 0.0066 0.0066>}]
 dyn_mushroom_02 [{shapeType=Capsule, length=6.1104, radius=4.2461, center=<1.9399 0.0017 0.0262>, rotation=<rot 30.4961 92.1955 -89.9992>}]
}
constraints [
 {bone1Name=dyn_mushroom_01, bone2Name=Melee_TwoHanded_Handle, swing1LimitDegrees=15, swing2LimitDegrees=15, twistLimitDegrees=1}
 {bone1Name=dyn_mushroom_02, bone2Name=dyn_mushroom_01, swing1LimitDegrees=15, swing2LimitDegrees=15, twistLimitDegrees=1}
 {
  bone1Name=dyn_leaves_01
  bone2Name=Melee_TwoHanded_Handle
  swing1LimitDegrees=25
  swing2LimitDegrees=25
  twistLimitDegrees=45
  twistMotion=Locked
 }
 {
  bone1Name=dyn_leaves_02
  bone2Name=Melee_TwoHanded_Handle
  swing1LimitDegrees=25
  swing2LimitDegrees=25
  twistLimitDegrees=45
  twistMotion=Locked
 }
 {
  bone1Name=dyn_leaves_03
  bone2Name=Melee_TwoHanded_Handle
  swing1LimitDegrees=25
  swing2LimitDegrees=25
  twistLimitDegrees=45
  twistMotion=Locked
 }
 {
  bone1Name=dyn_leaves_04
  bone2Name=Melee_TwoHanded_Handle
  swing1LimitDegrees=25
  swing2LimitDegrees=25
  twistLimitDegrees=45
  twistMotion=Locked
 }
 {
  bone1Name=dyn_leaves_05
  bone2Name=Melee_TwoHanded_Handle
  swing1LimitDegrees=25
  swing2LimitDegrees=25
  twistLimitDegrees=45
  twistMotion=Locked
 }
 {
  bone1Name=dyn_leaves_06
  bone2Name=Melee_TwoHanded_Handle
  swing1LimitDegrees=25
  swing2LimitDegrees=25
  twistLimitDegrees=45
  twistMotion=Locked
 }
 {
  bone1Name=dyn_leaves_07
  bone2Name=Melee_TwoHanded_Handle
  swing1LimitDegrees=25
  swing2LimitDegrees=25
  twistLimitDegrees=45
  twistMotion=Locked
 }
 {
  bone1Name=dyn_leaves_08
  bone2Name=Melee_TwoHanded_Handle
  swing1LimitDegrees=25
  swing1Motion=Locked
  swing2LimitDegrees=25
  twistLimitDegrees=45
  twistMotion=Locked
 }
]
